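# Physics asset for the Spartan Future backpack cosmetic, defining collision bodies and joint limits for its dangling parts.
# A box covers the pack, with grenade, handle and bandolier capsules each constrained to it and a box grenade pin hanging from the grenade.
# Three strap capsules chain from the spherical dyn_strap_MSTR body, which has no constraint to the pack here.
# Twist is locked or narrow throughout, so motion comes mainly from swing.
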PhysicsAsset /BRCosmetics/Accessories/FORT_Backpacks/Backpack_F_MED_Spartan_Future/Meshes/F_MED_Spartan_Future_Pack_Physics
bodies {
 dyn_bandolier_MSTR [{shapeType=Capsule, length=6.505, radius=2.005, center=<3.7325 -0.6915 1.647>, rotation=<rot -4.828 15.0698 8.7047>}]
 dyn_grenade [{shapeType=Capsule, length=7.9539, radius=3.2224, center=<1.7285 -0.8232 0.3971>, rotation=<rot -5.3387 -90 -1.9893>}]
 dyn_grenadePin [{shapeType=Box, extentX=2.9084, extentY=3.2885, extentZ=0.5, center=<1.5585 0.1527 0>}]
 dyn_handle [{shapeType=Capsule, length=2.6576, radius=0.8456, center=<0.8746 0.7062 -0.0627>, rotation=<rot -3.1762 -89.9999 4.7636>}]
 dyn_strap_01 [{shapeType=Capsule, length=0.6925, radius=0.88, center=<1.9145 0 0>}]
 dyn_strap_02 [{shapeType=Capsule, length=0.6133, radius=0.9076, center=<1.8421 -0.0001 0>}]
 dyn_strap_03 [{shapeType=Capsule, length=0.9029, radius=0.769, center=<1.0716 -0.0001 0>}]
 dyn_strap_MSTR [{shapeType=Sphere, radius=0.505}]
 pack [{shapeType=Box, extentX=9.6228, extentY=17.0702, extentZ=28.6362, center=<1.2166 -3.8216 -4.1686>, rotation=<rot -13.6456 -91.3749 28.6706>}]
}
constraints [
 {bone1Name=dyn_grenadePin, bone2Name=dyn_grenade, swing1LimitDegrees=40, swing2LimitDegrees=40, twistLimitDegrees=5}
 {bone1Name=dyn_grenade, bone2Name=pack, swing1LimitDegrees=5, swing2LimitDegrees=5, twistLimitDegrees=4}
 {bone1Name=dyn_strap_01, bone2Name=dyn_strap_MSTR, swing1LimitDegrees=20, swing2LimitDegrees=50, twistLimitDegrees=0}
 {bone1Name=dyn_strap_02, bone2Name=dyn_strap_01, swing1LimitDegrees=25, swing2LimitDegrees=50, twistLimitDegrees=0}
 {bone1Name=dyn_strap_03, bone2Name=dyn_strap_02, swing1LimitDegrees=20, swing2LimitDegrees=50, twistLimitDegrees=0}
 {bone1Name=dyn_handle, bone2Name=pack, swing1LimitDegrees=25, swing2LimitDegrees=0, twistLimitDegrees=0}
 {bone1Name=dyn_bandolier_MSTR, bone2Name=pack, swing1LimitDegrees=40, swing2LimitDegrees=15, twistLimitDegrees=0}
]
